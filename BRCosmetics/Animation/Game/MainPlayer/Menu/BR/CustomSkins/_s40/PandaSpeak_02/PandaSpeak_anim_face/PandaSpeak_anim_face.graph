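# Level Sequence for the face animation of the PandaSpeak_02 custom-skin menu animation, at 30 fps.
# A spawned FortPlayerMannequin plays a skeletal animation under a legacy face Control Rig, whose visible keys only move and scale the master control.
# Its child Head component carries a face bone-tweaker rig with no channels shown.
# A Head material track offsets the pupils horizontally, holding the left and animating the right.
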LevelSequence /BRCosmetics/Animation/Game/MainPlayer/Menu/BR/CustomSkins/_s40/PandaSpeak_02/PandaSpeak_anim_face/PandaSpeak_anim_face
bindings [
 FortPlayerMannequin {
  id=#1
  possessed_class=FortPlayerMannequin
  tracks [
   {type=MovieSceneSpawnTrack, sections [{type=MovieSceneSpawnSection, channels [None {type=MovieSceneScriptingBoolChannel, default=true}]}]}
   {
    type=MovieScene3DTransformTrack
    sections [{
      type=MovieScene3DTransformSection
      channels [
       Location.X {type=MovieSceneScriptingDoubleChannel, default=0, thinned_from=18, keys [[0,0.1481],[19,0.1481],[20,0],[325,0]]}
       Location.Y {type=MovieSceneScriptingDoubleChannel, default=0, const [0,1]}
       Location.Z {type=MovieSceneScriptingDoubleChannel, default=0, const [0,1]}
       Rotation.X {type=MovieSceneScriptingDoubleChannel, default=0, const [0,1]}
       Rotation.Y {type=MovieSceneScriptingDoubleChannel, default=0, const [0,1]}
       Rotation.Z {type=MovieSceneScriptingDoubleChannel, default=0, const [0,1]}
       Scale.X {type=MovieSceneScriptingDoubleChannel, default=1, const [1,1]}
       Scale.Y {type=MovieSceneScriptingDoubleChannel, default=1, const [1,1]}
       Scale.Z {type=MovieSceneScriptingDoubleChannel, default=1, const [1,1]}
      ]
     }]
   }
   {type=MovieSceneSkeletalAnimationTrack, sections [{type=MovieSceneSkeletalAnimationSection, end_frame=400, start_frame=0}]}
   {
    type=MovieSceneControlRigParameterTrack
    display_name=Fortnite_Legacy_Face_CtrlRig
    sections [{
      type=MovieSceneControlRigParameterSection
      channels [
       Flipbook {type=MovieSceneScriptingBoolChannel, default=false, const [false,1]}
       Face_Ctrl_Master.Location.X {type=MovieSceneScriptingFloatChannel, default=20.2381, const [80,1]}
       Face_Ctrl_Master.Location.Y {type=MovieSceneScriptingFloatChannel, default=0.0958, keys [[409,-0.7108],[418,-0.2976]]}
       Face_Ctrl_Master.Location.Z {type=MovieSceneScriptingFloatChannel, default=11.588, keys [[409,-85.998],[418,-35.9988]]}
       Face_Ctrl_Master.Rotation.X {type=MovieSceneScriptingFloatChannel, default=0, const [0,1]}
       Face_Ctrl_Master.Rotation.Y {type=MovieSceneScriptingFloatChannel, default=0, const [0,1]}
       Face_Ctrl_Master.Rotation.Z {type=MovieSceneScriptingFloatChannel, default=0, const [0,1]}
       Face_Ctrl_Master.Scale.X {type=MovieSceneScriptingFloatChannel, default=1, keys [[409,8.9773],[418,6]]}
       Face_Ctrl_Master.Scale.Y {type=MovieSceneScriptingFloatChannel, default=1, keys [[409,8.9773],[418,6]]}
       Face_Ctrl_Master.Scale.Z {type=MovieSceneScriptingFloatChannel, default=1, keys [[409,8.9773],[418,6]]}
       gui_follow_head {type=MovieSceneScriptingFloatChannel, default=1, const [1,1]}
       Weight {type=MovieSceneScriptingFloatChannel, default=1}
      ]
     }]
   }
  ]
 }
 Head {
  id=#2
  parent_id=#1
  possessed_class=SkeletalMeshComponentBudgeted
  tracks [
   {
    type=MovieSceneControlRigParameterTrack
    display_name=Fortnite_Face_BoneTweaker_CtrlRig
    sections [{type=MovieSceneControlRigParameterSection}]
   }
   {
    type=MovieSceneComponentMaterialTrack
    display_name=Material Slot: Head
    sections [{
      type=MovieSceneComponentMaterialParameterSection
      end_frame=400
      start_frame=0
      channels [
       L_pupil_offsetX {type=MovieSceneScriptingFloatChannel, default=0, const [0.1,8]}
       R_pupil_offsetX {type=MovieSceneScriptingFloatChannel, default=-1.5116, thinned_from=8, keys [[0,-0.1],[24,-0.1],[30,0.04],[82,0.04],[86,-0.1],[399,-0.1]]}
      ]
     }]
   }
  ]
 }
]
display_rate {denominator=1, numerator=30}
tick_resolution {denominator=1, numerator=24000}
tracks []
